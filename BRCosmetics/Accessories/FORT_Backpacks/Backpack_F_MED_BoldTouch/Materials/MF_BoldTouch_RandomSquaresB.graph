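# Material function for the BoldTouch backpack that produces animated random squares with a grid-line overlay.
# A grayscale mosaic texture offsets scaled Time, and the result serves as noise texture UVs, giving each mosaic value its own phase in a changing brightness.
# SmoothStep remaps the noise between SquaresMin and SquaresMax, then the inverted lines texture masks it.
# Outputs provide the masked squares, the lines, and their sum.
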MaterialFunction /BRCosmetics/Accessories/FORT_Backpacks/Backpack_F_MED_BoldTouch/Materials/MF_BoldTouch_RandomSquaresB
graph {nodes [
  n1 = FunctionInput(Preview=Constant() {r=0.1}) {bUsePreviewValueAsDefault=true, inputName=SquaresMin, inputType=FunctionInput_Scalar, previewValue=<0.1 0 0 1>}
  n2 = FunctionInput(Preview=Constant() {r=0.9}) {bUsePreviewValueAsDefault=true, inputName=SquaresMax, inputType=FunctionInput_Scalar, previewValue=<0.9 0 0 1>}
  n3 = Time()
  n4 = ScalarParameter() {defaultValue=0.1, parameterName=TimeScale, sortPriority=35}
  n5 = TextureSampleParameter2D() {
   parameterName=RandomSquares
   texture=/BRCosmetics/Accessories/FORT_Backpacks/Backpack_F_MED_BoldTouch/Textures/T_BoldTouch_Backpack_Grayscale
  }
  n6 = TextureSampleParameter2D(UVs=Add(A=Multiply(A=@n3, B=@n4) {constB=0.1}, B=NamedRerouteDeclaration(None=@n5.R) {name=RandomMosaic, nodeColor=<rgba 0.7109 1 0 1>})) {
   group=BoldTouch_RandomSquares
   parameterName=NoiseTexture
   samplerType=SAMPLERTYPE_Grayscale
   sortPriority=33
   texture=/Game/Global/Textures/Default/Noise/LowResBlurredNoiseNormalized
  }
  n7 = TextureSampleParameter2D() {
   group=BoldTouch_RandomSquares
   parameterName=RandomSquaresLines
   sortPriority=34
   texture=/BRCosmetics/Accessories/FORT_Backpacks/Backpack_F_MED_BoldTouch/Textures/T_BoldTouch_Backpack_Lines
  }
  n8 = Multiply(
   A=SmoothStep(Min=@n1, Max=@n2, Value=@n6.R) {constMax=0.9, constMin=0.1}
   B=NamedRerouteDeclaration(None=OneMinus(None=@n7.R)) {name=GridLines, nodeColor=<rgba 0.9688 0 1 1>}
  )
  FunctionOutput(None=Add(A=@n8, B=@n7.R)) {outputName=SquaresAndLines}
  FunctionOutput(None=@n8) {outputName=RandomSquares}
  FunctionOutput(None=@n7.R) {outputName=Lines}
 ]}
