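# Material function for the BoldTouch backpack that overlays an animated grid of colored squares and lines on incoming material attributes.
# MF_BoldTouch_RandomSquaresB supplies squares and lines for tinting, alongside inline tiled-UV, hash and cosine-over-time grid math.
# Two panning packed textures multiply into noise for breaking up edges, and a powered Fresnel term shapes the glow.
# The masked result adds to Emissive and lerps into Base Color.
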MaterialFunction /BRCosmetics/Accessories/FORT_Backpacks/Backpack_F_MED_BoldTouch/Materials/MF_BoldTouchBackpackGrid
graph {nodes [
  Add(A=NamedRerouteUsage(), B=Multiply(A=NamedRerouteUsage(), B=NamedRerouteUsage()))
  n1 = GetMaterialAttributes(None=NamedRerouteUsage()) {attributeGetTypes [#1, #2]}
  n2 = NamedRerouteUsage()
  n3 = NamedRerouteUsage()
  n4 = Multiply(A=NamedRerouteUsage(), B=ScalarParameter() {defaultValue=0.5, group=BoldTouchBackpackTechArt, parameterName=NoiseEdgeBreakupMult})
  n5 = ScalarParameter() {defaultValue=0.25, group=BoldTouchBackpackTechArt, parameterName=NoiseEdgeBreakupBias}
  n6 = ScalarParameter() {defaultValue=3, group=BoldTouchBackpackTechArt, parameterName=ExponentIn_FresnelMask}
  n7 = Power(
   Base=MaterialFunctionCall(A=Constant(), B=Add(A=@n4, B=@n5), C=Constant() {r=1}, Alpha=Fresnel(ExponentIn=@n6)) {materialFunction=&Lerp_3Color}.Result
   Exp=ScalarParameter() {defaultValue=2, group=BoldTouchBackpackTechArt, parameterName=FresnelPower}
  )
  n8 = ScalarParameter() {defaultValue=2, group=BoldTouchBackpackTechArt, parameterName=FresnelMultiplyAfterPower}
  n9 = Multiply(A=Multiply(A=Add(A=@n2, B=@n3), B=Multiply(A=@n7, B=@n8) {constB=2}), B=ComponentMask(None=NamedRerouteUsage()) {r=true})
  n10 = Multiply(A=@n9) {constB=4}
  FunctionOutput(None=SetMaterialAttributes!(
    MaterialAttributes=@n1.MaterialAttributes
    EmissiveColor=Add(A=@n1.EmissiveColor, B=@n9)
    BaseColor=LinearInterpolate(A=@n1.BaseColor, B=@n10, Alpha=@n9)
   ) {attributeSetTypes [#1, #2]})
  MaterialFunctionCall!(
   SquaresMax=ScalarParameter() {group=BoldTouchBackpackTechArt, parameterName=SquaresMinValue}
   SquaresMin=ScalarParameter() {defaultValue=1, group=BoldTouchBackpackTechArt, parameterName=SquaresMaxValue}
  ) {materialFunction=&/MF_BoldTouch_RandomSquaresB}
  NamedRerouteDeclaration(None=FunctionInput() {bUsePreviewValueAsDefault=true, inputName=IN_TechArtMask}) {name=IN_TechArtMask, nodeColor=<rgba 1 0.375 0 1>}
  NamedRerouteDeclaration(None=FunctionInput() {bUsePreviewValueAsDefault=true, inputName=IN_MaterialAttributes, inputType=FunctionInput_MaterialAttributes}) {name=IN_MaterialAttributes, nodeColor=<rgba 0.0547 0 1 1>}
  NamedRerouteDeclaration(None=MaterialFunctionCall(Vector 3=Constant3Vector() {constant=<rgba 0 1 2 0>}) {materialFunction=/Engine/Functions/Engine_MaterialFunctions02/Utility/VectorLength}.V3 Length) {name=VectorLength, nodeColor=<rgba 0.1953 0 1 1>}
  n11 = NamedRerouteUsage()
  NamedRerouteDeclaration(None=Add(A=NamedRerouteUsage(), B=Multiply(B=@n11) {constA=2})) {name=TimePlusId, nodeColor=<rgba 0.125 1 0 1>}
  n12 = TextureCoordinate()
  n13 = AppendVector(
   A=ScalarParameter() {defaultValue=24, group=BoldTouchBackpackTechArt, parameterName=UTilingGridLines}
   B=ScalarParameter() {defaultValue=24, group=BoldTouchBackpackTechArt, parameterName=VTilingGridLines, sortPriority=35}
  )
  NamedRerouteDeclaration(None=Frac(None=Multiply(A=@n12, B=@n13))) {name=TiledUVsForGridLines, nodeColor=<rgba 0.875 0 1 1>}
  n14 = Multiply(A=NamedRerouteUsage()) {constB=2}
  NamedRerouteDeclaration(None=Abs(None=Subtract(A=@n14))) {name=Square, nodeColor=<rgba 0 1 0.5547 1>}
  NamedRerouteDeclaration(None=Step(Y=ScalarParameter() {defaultValue=0.95, group=BoldTouchBackpackTechArt, parameterName=GridLineSize}, X=NamedRerouteUsage()) {constY=0.95}) {name=SharpSquare, nodeColor=<rgba 1 0.2344 0 1>}
  n15 = TextureCoordinate()
  n16 = AppendVector(
   A=ScalarParameter() {defaultValue=24, group=BoldTouchBackpackTechArt, parameterName=UTiling}
   B=ScalarParameter() {defaultValue=24, group=BoldTouchBackpackTechArt, parameterName=VTiling, sortPriority=35}
  )
  NamedRerouteDeclaration(None=Add(A=Multiply(A=@n15, B=@n16), B=Constant2Vector() {g=-0.5, r=-0.5})) {name=UVs, nodeColor=<rgba 1 0.8488 0 1>}
  n17 = MaterialFunctionCall(Float2=NamedRerouteUsage()) {materialFunction=&BreakOutFloat2Components}
  NamedRerouteDeclaration(None=Add(A=@n17.R, B=@n17.R)) {name=SquareResult, nodeColor=<rgba 0 0.0859 1 1>}
  n18 = TextureCoordinate()
  n19 = Multiply(A=@n18)
  n20 = Time()
  n21 = AppendVector(
   A=ScalarParameter() {defaultValue=0.2, group=BoldTouchBackpackTechArt, parameterName=LowFreqNoiseUSpeed}
   B=ScalarParameter() {defaultValue=0.62, group=BoldTouchBackpackTechArt, parameterName=LowFreqNoiseVSpeed}
  )
  n22 = Multiply(A=@n18)
  n23 = Time()
  n24 = AppendVector(
   A=ScalarParameter() {defaultValue=0.2, group=BoldTouchBackpackTechArt, parameterName=HighFreqNoiseUSpeed}
   B=ScalarParameter() {defaultValue=0.62, group=BoldTouchBackpackTechArt, parameterName=HighFreqNoiseVSpeed}
  )
  n25 = Multiply(
   A=TextureSample(
    UVs=Panner(Coordinate=@n19, Time=@n20, Speed=@n21)
    Tex=TextureObjectParameter() {parameterName=NoiseTexture, texture=/BRCosmetics/Accessories/FORT_Backpacks/Backpack_F_MED_BoldTouch/Textures/T_PackedSquaresA}
   ).G
   B=TextureSample(
    UVs=Panner(Coordinate=@n22, Time=@n23, Speed=@n24)
    Tex=TextureObjectParameter() {parameterName=NoiseTextureB, texture=/BRCosmetics/Accessories/FORT_Backpacks/Backpack_F_MED_BoldTouch/Textures/T_PackedDotsA}
   ).B
  )
  NamedRerouteDeclaration(None=SmoothStep(
    Min=ScalarParameter() {group=BoldTouchBackpackTechArt, parameterName=MaskSStepMin}
    Max=ScalarParameter() {defaultValue=1, group=BoldTouchBackpackTechArt, parameterName=MaskSStepMax}
    Value=Saturate(None=@n25)
   )) {name=NoiseTotalForEdge, nodeColor=<rgba 0 1 0.6016 1>}
  n26 = Add(A=NamedRerouteUsage(), B=Constant2Vector() {g=0.5, r=0.5}) {constB=0.5}
  n27 = Sine(None=DotProduct(A=Floor(None=@n26), B=Constant2Vector() {g=17.81, r=113.1}))
  n28 = Cosine(None=Add(A=NamedRerouteUsage(), B=Multiply(A=@n27) {constB=43758.5469}))
  NamedRerouteDeclaration(None=Add(B=Multiply(B=@n28) {constA=0.5}) {constA=0.5}) {name=id, nodeColor=<rgba 0 0.5078 1 1>}
  n29 = NamedRerouteUsage()
  n30 = Cosine(None=Divide(A=Add(A=@n29) {constB=0}, B=NamedRerouteUsage()))
  NamedRerouteDeclaration(None=Add(A=ScalarParameter() {defaultValue=0.25, parameterName=Amount of Lit Squares, sliderMax=1, sliderMin=-0.5}, B=Multiply(B=@n30) {constA=0.5}) {constA=-0.5}) {name=coRed, nodeColor=<rgba 1 0 0.0044 1>}
  n31 = ScalarParameter() {group=BoldTouchBackpackTechArt, parameterName=SquaresMinValue}
  n32 = ScalarParameter() {defaultValue=1, group=BoldTouchBackpackTechArt, parameterName=SquaresMaxValue}
  NamedRerouteDeclaration(None=MaterialFunctionCall(
    A=Constant3Vector()
    B=VectorParameter() {group=BoldTouchBackpackTechArt, parameterName=SquareColorA, defaultValue=<rgba 0.7011 0.3419 0.8388 1>}.RGB
    C=VectorParameter() {group=BoldTouchBackpackTechArt, parameterName=SquareColorB, defaultValue=<rgba 0.8963 0.2462 0.7991 1>}.RGB
    Alpha=MaterialFunctionCall(SquaresMax=@n31, SquaresMin=@n32) {materialFunction=&MF_BoldTouch_RandomSquaresB}.RandomSquares
   ) {materialFunction=&Lerp_3Color}.Result) {name=RandomSquaresColor, nodeColor=<rgba 1 0.5156 0 1>}
  NamedRerouteDeclaration(None=Multiply(A=Time(), B=ScalarParameter() {defaultValue=0.2, group=BoldTouchBackpackTechArt, parameterName=TimeMultiplier}) {constB=0.05}) {name=TimeMult, nodeColor=<rgba 0.7578 0 1 1>}
  n33 = VectorParameter() {group=BoldTouchBackpackTechArt, parameterName=SquareLineColor, defaultValue=<rgba 0.8963 0.2462 0.7991 1>}
  NamedRerouteDeclaration(None=Multiply(
    A=@n33.RGB
    B=MaterialFunctionCall() {materialFunction=&MF_BoldTouch_RandomSquaresB}.Lines
   )) {name=LineColor, nodeColor=<rgba 1 0.5156 0 1>}
  n34 = NamedRerouteUsage()
  NamedRerouteDeclaration(None=MaterialFunctionCall(
    A=Constant3Vector()
    B=VectorParameter() {group=BoldTouchBackpackTechArt, parameterName=SquareColorA, defaultValue=<rgba 0.7011 0.3419 0.8388 1>}.RGB
    C=VectorParameter() {group=BoldTouchBackpackTechArt, parameterName=SquareColorB, defaultValue=<rgba 0.8963 0.2462 0.7991 1>}.RGB
    Alpha=NamedRerouteDeclaration(None=@n34) {name=SquaresIntensity, nodeColor=<rgba 1 0.7266 0 1>}
   ) {materialFunction=&Lerp_3Color}.Result) {name=Color, nodeColor=<rgba 1 0.5156 0 1>}
 ]}
