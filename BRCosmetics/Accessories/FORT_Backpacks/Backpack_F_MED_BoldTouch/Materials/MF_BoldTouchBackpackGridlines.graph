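# Material function for the BoldTouch backpack that generates an animated, randomized grid-cell pattern from tiled UVs.
# Scaled UVs and time build a position that is floored into cell indices, with MF_Hash supplying per-cell animation offsets and random values.
# A cosine pulse and step thresholds shape the color, which is output through a 1/2.2 gamma power curve.
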MaterialFunction /BRCosmetics/Accessories/FORT_Backpacks/Backpack_F_MED_BoldTouch/Materials/MF_BoldTouchBackpackGridlines
graph {nodes [
  n1 = Multiply(
   A=TextureCoordinate()
   B=AppendVector(
    A=ScalarParameter() {defaultValue=1, group=BoldTouchBackpackTechArt, parameterName=UTiling}
    B=ScalarParameter() {defaultValue=1, group=BoldTouchBackpackTechArt, parameterName=VTiling, sortPriority=35}
   )
  )
  Add(A=@n1, B=Constant2Vector() {g=-0.5, r=-0.5})
  n2 = Constant() {r=1}
  n3 = Constant() {r=2.2}
  FunctionOutput(None=Power(Base=NamedRerouteUsage(), Exp=Divide(A=@n2, B=@n3)))
  NamedRerouteDeclaration(None=Constant() {r=2}) {name=GridDimensions, nodeColor=<rgba 1 0.0234 0 1>}
  NamedRerouteDeclaration(None=Multiply(A=Time(), B=ScalarParameter() {defaultValue=0.02, parameterName=TimeMultiplier})) {name=TimeAdjusted, nodeColor=<rgba 0.6406 0 1 1>}
  n4 = NamedRerouteUsage()
  n5 = NamedRerouteUsage()
  NamedRerouteDeclaration(None=Floor(None=Add(A=@n4, B=@n5) {constB=0})) {name=Animation, nodeColor=<rgba 0 1 0.3203 1>}
  n6 = AppendVector(A=NamedRerouteUsage(), B=NamedRerouteUsage())
  NamedRerouteDeclaration(None=ComponentMask(None=MaterialFunctionCall(Vec3=@n6) {materialFunction=&MF_Hash}.Vec3) {r=true}) {name=Rnd, nodeColor=<rgba 1 0 0 1>}
  n7 = MaterialFunctionCall(Multiplier=Constant() {r=2}) {materialFunction=&Pi}
  n8 = NamedRerouteUsage()
  n9 = Cosine(None=Subtract(A=Multiply(A=@n7.Result, B=@n8), B=MaterialFunctionCall() {materialFunction=&Pi}.Result))
  NamedRerouteDeclaration(None=Add(A=Multiply(A=@n9) {constB=0.5}) {constB=0.5}) {name=q, nodeColor=<rgba 0 0.8828 1 1>}
  MaterialExpressionNamedRerouteUsage_13 = NamedRerouteUsage()
  n10 = Convert(None=@MaterialExpressionNamedRerouteUsage_13) {
   nodeName=Break Float2
   convertInputs [{
     type=Vector2
     expressionInput {
      expression=&self:MaterialExpressionNamedRerouteUsage_13
      inputName=None
      mask=0
      maskA=0
      maskB=0
      maskG=0
      maskR=0
      outputIndex=0
     }
    }]
   convertOutputs [{type=Scalar}, {type=Scalar}]
  }
  n11 = NamedRerouteUsage()
  NamedRerouteDeclaration(None=Multiply(A=Multiply(A=@n10, B=@n10), B=Step(Y=@n11) {constX=0.5, constY=0.5})) {name=Color, nodeColor=<rgba 0.4062 1 0 1>}
  NamedRerouteDeclaration(None=Multiply(A=Time(), B=ScalarParameter() {defaultValue=0.2, parameterName=TimeMultiplierB})) {name=TimeAdjusted_1, nodeColor=<rgba 0.6406 0 1 1>}
  NamedRerouteDeclaration(None=@n1) {name=UVs, nodeColor=<rgba 1 0.8488 0 1>}
  n12 = NamedRerouteUsage()
  n13 = NamedRerouteUsage()
  NamedRerouteDeclaration(None=Multiply(A=NamedRerouteUsage(), B=Add(A=@n12, B=@n13))) {name=position, nodeColor=<rgba 1 0 0.7969 1>}
  NamedRerouteDeclaration(None=Floor(None=NamedRerouteUsage())) {name=CellIndex, nodeColor=<rgba 1 0.5625 0 1>}
  n14 = NamedRerouteUsage()
  n15 = AppendVector(A=NamedRerouteUsage(), B=Constant())
  NamedRerouteDeclaration(None=ComponentMask(None=MaterialFunctionCall(Seed=@n14, Vec3=@n15) {materialFunction=&MF_Hash}.Vec3) {r=true}) {name=AnimationOffset, nodeColor=<rgba 1 0.9609 0 1>}
 ]}
